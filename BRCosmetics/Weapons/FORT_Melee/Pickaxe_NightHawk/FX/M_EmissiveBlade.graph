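# Unlit, translucent, two-sided emissive blade material for the NightHawk pickaxe's effects, enabled for Niagara mesh particles and ribbons.
# The Swipe dynamic parameter offsets inverted V from UV channel 1, and a narrow SmoothStep turns that into a soft-edged moving mask.
# That mask times particle alpha blends particle color toward the Emissive Color parameter.
# The result feeds the shared MF_FX_EmissiveAdjustments function.
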Material /BRCosmetics/Weapons/FORT_Melee/Pickaxe_NightHawk/FX/M_EmissiveBlade
graph {
 properties {
  bUsedWithNiagaraMeshParticles=true
  bUsedWithNiagaraRibbons=true
  blendMode=BLEND_Translucent
  pixelDepthOffsetMode=PDOM_Legacy
  shadingModel=MSM_Unlit
  twoSided=true
  usedShadingModels=Unlit
 }
 nodes [
  n1 = ParticleColor()
  n2 = ComponentMask(None=TextureCoordinate() {coordinateIndex=1}) {g=true}
  n3 = SmoothStep(Value=Add(A=OneMinus(None=@n2), B=DynamicParameter() {defaultValue=<rgba 0 1 1 1>, paramNames [Swipe,Param2,Param3,Param4]}.Swipe)) {constMax=1.25, constMin=1}
  MaterialFunctionCall(Emissive=LinearInterpolate(A=@n1.RGB, B=VectorParameter() {parameterName=Emissive Color, defaultValue=<rgba 5 5 5 1>}.RGB, Alpha=Multiply(A=@n1.A, B=@n3))) {materialFunction=&MF_FX_EmissiveAdjustments}
 ]
}
