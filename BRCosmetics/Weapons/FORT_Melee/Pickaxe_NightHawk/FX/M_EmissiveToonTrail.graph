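# Unlit, two-sided translucent material for Niagara ribbon trails on the NightHawk pickaxe, named as an emissive toon trail.
# An inverted, smoothstepped UV sphere mask with adjustable radius and hardness, scaled by particle alpha, shapes the trail.
# That mask blends particle color toward a bright Emissive Color, and is also multiplied by a U-axis fade driven by the Leading Fade dynamic parameter.
# A separate emissive-adjustment function also takes particle color.
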Material /BRCosmetics/Weapons/FORT_Melee/Pickaxe_NightHawk/FX/M_EmissiveToonTrail
graph {
 properties {
  bUsedWithNiagaraRibbons=true
  blendMode=BLEND_Translucent
  pixelDepthOffsetMode=PDOM_Legacy
  shadingModel=MSM_Unlit
  twoSided=true
  usedShadingModels=Unlit
 }
 nodes [
  n1 = ParticleColor()
  n2 = SphereMask(
   A=TextureCoordinate()
   B=Constant2Vector() {g=1, r=1}
   Radius=ScalarParameter() {defaultValue=1.1, parameterName=Radius, sliderMax=1.5}
   Hardness=ScalarParameter() {defaultValue=0.9, parameterName=Hardness, sliderMax=1}
  ) {attenuationRadius=1.1, hardnessPercent=90}
  n3 = Multiply(
   A=@n1.A
   B=SmoothStep(
    Min=ScalarParameter() {parameterName=Min, sliderMax=1}
    Max=ScalarParameter() {defaultValue=1, parameterName=Max, sliderMax=1}
    Value=OneMinus(None=@n2)
   )
  )
  LinearInterpolate(A=@n1.RGB, B=VectorParameter() {parameterName=Emissive Color, defaultValue=<rgba 8 8 8 1>}.RGB, Alpha=@n3)
  MaterialFunctionCall(Emissive=@n1.RGB) {materialFunction=&MF_FX_EmissiveAdjustments}
  n4 = TextureCoordinate()
  Multiply(
   A=@n3
   B=SmoothStep(
    Max=DynamicParameter() {defaultValue=<rgba 0 1 1 1>, paramNames [Leading Fade, Param2, Param3, Param4]}.Leading Fade
    Value=ComponentMask(None=@n4) {r=true}
   ) {constMax=0}
  )
 ]
}
