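# Physics asset for the NightHawk two-handed pickaxe mesh, giving it a simulated dangling cord.
# The handle bone carries two capsules, a long thin one and a wide one offset from it.
# Four cord bones each carry a small capsule and are constrained in a chain hanging from the handle.
# Angular limits keep twist tight throughout, with the root link narrow on its second swing axis and the outer links moderately limited.
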PhysicsAsset /BRCosmetics/Weapons/FORT_Melee/Pickaxe_NightHawk/Meshes/NightHawk_Axe_Physics
bodies {
 Melee_TwoHanded_Handle [
  {shapeType=Capsule, length=73.1482, radius=4.4871, center=<0 0.4978 -0.6621>}
  {shapeType=Capsule, length=42, radius=17, shapeName=Melee_TwoHanded_Handle_capsule, center=<12.1584 30.2609 0>, rotation=<rot -0.0001 10 90>}
 ]
 dyn_cord_1 [{shapeType=Capsule, length=19.5, radius=2, shapeName=dyn_cord_1_capsule, center=<-0.588 2.671 -0.2682>}]
 dyn_cord_2 [{shapeType=Capsule, length=19, radius=2, shapeName=dyn_cord_2_capsule, center=<-1.3689 2.3023 0.1203>}]
 dyn_cord_3 [{shapeType=Capsule, length=14, radius=2, shapeName=dyn_cord_3_capsule, center=<-0.4713 0.5117 0>}]
 dyn_cord_4 [{shapeType=Capsule, length=8, radius=2, shapeName=dyn_cord_4_capsule, center=<0.2993 -0.3164 -0.4104>}]
}
constraints [
 {bone1Name=dyn_cord_2, bone2Name=dyn_cord_1, swing1LimitDegrees=60, swing2LimitDegrees=30, twistLimitDegrees=5}
 {bone1Name=dyn_cord_1, bone2Name=Melee_TwoHanded_Handle, swing1LimitDegrees=60, swing2LimitDegrees=4, twistLimitDegrees=2}
 {bone1Name=dyn_cord_3, bone2Name=dyn_cord_2, swing1LimitDegrees=30, swing2LimitDegrees=25, twistLimitDegrees=5}
 {bone1Name=dyn_cord_4, bone2Name=dyn_cord_3, swing1LimitDegrees=30, swing2LimitDegrees=25, twistLimitDegrees=5}
]
