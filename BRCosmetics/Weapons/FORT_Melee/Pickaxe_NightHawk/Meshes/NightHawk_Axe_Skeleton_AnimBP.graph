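# Animation Blueprint for the NightHawk pickaxe skeleton that adds secondary motion to its dyn_cord bones.
# The cached input pose feeds three alternative simulations: RigidBody, chained AnimDynamics, and a half-alpha Trail controller with a planar limit.
# bRigidBodyEnabled selects RigidBody, otherwise bAnimDynamicsEnabled chooses AnimDynamics over the Trail fallback.
# On the RigidBody path, bIsPlayingMeleeAnim blends in an additive yaw on dyn_cord_MSTR before simulation.
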AnimBlueprint /BRCosmetics/Weapons/FORT_Melee/Pickaxe_NightHawk/Meshes/NightHawk_Axe_Skeleton_AnimBP
target_skeleton=&/NightHawk_Axe_Skeleton
graphs [
 AnimGraph {nodes [
   n1 = Anim Dynamics - Bone: dyn_cord_1!(
    ComponentPose=Local To Component(LocalPose=Use cached pose 'PreSim'())
    Alpha=1
    ComponentLinearAccScale:Vector="0.3,0.3,0.3"
    ComponentLinearVelScale:Vector="0.3,0.3,0.3"
    ExternalForce="0,0,0"
   ) -> (Pose) {
    angularDampingOverride=0.9
    angularSpringConstant=400
    bAngularSpring=true
    bOverrideAngularDamping=true
    bUseGravityOverride=true
    linearDampingOverride=0.9
    linearSpringConstant=100
    simulationSpace=BoneRelative
    alphaBoolBlend {blendOption=Linear}
    boundBone {boneName=dyn_cord_1}
    componentAppliedLinearAccClamp=<3000 3000 3000>
    componentLinearAccScale=<0.3 0.3 0.3>
    componentLinearVelScale=<0.3 0.3 0.3>
    gravityOverride {z=-1000}
    physicsBodyDefinitions [{
      boundBone {boneName=dyn_cord_1}
      boxExtents=<7 2 7>
      constraintSetup {angularLimitsMax=<5 5 5>, angularLimitsMin=<-5 -5 -5>, angularTarget=<1 0 0>}
      localJointOffset=<1 1 0>
     }]
    relativeSpaceBone {boneName=b_AR_Root}
   }
   n2 = "Anim Dynamics - Chain: dyn_cord_2 -> dyn_cord_4"!(
    ComponentPose=@n1
    Alpha=1
    ComponentLinearAccScale:Vector="0.2,0.2,0.2"
    ComponentLinearVelScale:Vector="0.2,0.2,0.2"
    ExternalForce="0,0,0"
   ) -> (Pose) {
    angularDampingOverride=0.9
    angularSpringConstant=200
    bAngularSpring=true
    bChain=true
    bOverrideAngularDamping=true
    bUseGravityOverride=true
    linearDampingOverride=0.9
    linearSpringConstant=100
    simulationSpace=BoneRelative
    alphaBoolBlend {blendOption=Linear}
    boundBone {boneName=dyn_cord_2}
    chainEnd {boneName=dyn_cord_4}
    componentAppliedLinearAccClamp=<3000 3000 3000>
    componentLinearAccScale=<0.3 0.3 0.3>
    componentLinearVelScale=<0.3 0.3 0.3>
    gravityOverride {z=-1000}
    physicsBodyDefinitions [
     {
      boundBone {boneName=dyn_cord_2}
      boxExtents=<7 2 7>
      constraintSetup {angularLimitsMax=<20 20 20>, angularLimitsMin=<0 0 0>, angularTarget=<1 0 0>}
      localJointOffset=<-5 1 0>
     }
     {
      boundBone {boneName=dyn_cord_3}
      boxExtents=<7 2 7>
      constraintSetup {angularLimitsMax=<20 20 20>, angularLimitsMin=<-5 -5 -5>, angularTarget=<1 0 0>}
      localJointOffset=<-5 1 0>
     }
     {
      boundBone {boneName=dyn_cord_4}
      boxExtents=<7 2 7>
      constraintSetup {angularLimitsMax=<10 10 10>, angularLimitsMin=<-10 -10 -10>, angularTarget=<1 0 0>}
      localJointOffset=<-5 1 0>
     }
    ]
    relativeSpaceBone {boneName=b_AR_Root}
   }
   n3 = Component To Local(ComponentPose=@n2)
   n4 = Trail controller - Bone: dyn_cord_4(ComponentPose=Local To Component(LocalPose=Use cached pose 'PreSim'()), Alpha=0.5) {
    bReorientParentToChild=false
    bShowBaseMotion=false
    bShowLimit=false
    bUsePlanarLimit=true
    chainBoneAxis=Y
    chainLength=4
    relaxationSpeedScale=3
    baseJoint {boneName=dyn_cord_1}
    planarLimits [{drivingBone {boneName=dyn_cord_MSTR}, planeTransform {location=<-5 10 0>, rotation=<rot 90 0 0>, scale=<1 1.5 1>}}]
    rotationLimits [{limitMax=<180 180 180>}, {limitMax=<180 180 180>}, {limitMax=<180 180 180>}, {limitMax=<180 180 180>}]
    rotationOffsets [<0 0 0>, <0 0 0>, <0 0 0>, <0 0 0>]
    trailBone {boneName=dyn_cord_4}
    trailRelaxationSpeed {editorCurveData {keys [<at 0 10 RCIM_Linear>,<at 0.2525 6.6729 RCIM_Linear>,<at 0.5945 5.4697 RCIM_Linear>]}}
   }
   n5 = Component To Local(ComponentPose=@n4)
   n6 = Get bAnimDynamicsEnabled() -> (bAnimDynamicsEnabled:bool)
   n7 = "Transform (Modify) Bone - Bone: dyn_cord_MSTR"!(
    ComponentPose=Local To Component(LocalPose=Use cached pose 'PreSim'())
    bAlphaBoolEnabled:bool=Get bIsPlayingMeleeAnim()
    Translation="0,0,0"
    Rotation="0,30,0"
    Scale="1,1,1"
   ) -> (Pose) {
    alphaInputType=Bool
    rotationMode=BMM_Additive
    alphaBoolBlend {blendInTime=0.3, blendOption=ExpInOut, blendOutTime=0.3}
    alphaScaleBiasClamp {bInterpResult=true}
    boneToModify {boneName=dyn_cord_MSTR}
    rotation {yaw=30}
   }
   n8 = RigidBody(ComponentPose=@n7) {
    bOverrideWorldGravity=true
    overridePhysicsAsset=&/NightHawk_Axe_Physics
    alphaBoolBlend {blendInTime=0.3, blendOption=ExpInOut, blendOutTime=0.3}
    alphaScaleBiasClamp {bInterpResult=true, interpSpeedDecreasing=20, interpSpeedIncreasing=20}
    componentAppliedLinearAccClamp=<4000 4000 4000>
    componentLinearAccScale=<1 1 1>
    componentLinearVelScale=<1 1 1>
    overrideWorldGravity {z=-980}
   }
   Output Pose(Result=Blend Poses by bool(
     BlendPose_0=Component To Local(ComponentPose=@n8)
     BlendPose_1=Blend Poses by bool(BlendPose_0=@n3, BlendPose_1=@n5, bActiveValue=@n6) {blendType=Linear}
     bActiveValue=Get bRigidBodyEnabled()
    ) {blendType=Linear}) {layerGroup=None}
   Save cached pose 'PreSim'(Pose=Input Pose()) {cacheName=PreSim}
  ]}
 EventGraph {nodes [Event BlueprintUpdateAnimation()]}
]
variables [
 Component Linear Acc Scale {type=Vector}
 Component Linear Vel Scale {type=Vector}
 Delta Time X {type=number}
 FPS Compensated Alpha {type=number}
 FPS Compensated Transform Alpha {type=number}
]
